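# Control Rig for the FrozenTouch backpack that poses its dynamic spine bone chain according to the character's movement state.
# Three chained Modify Transforms steps apply small rotations and offsets, weighted by the crouch, running and tactical sprint variables through interpolated alphas.
# The crouch weight is the absolute value of one minus the crouch value, so that offset is fully applied when standing and fades out in a crouch.
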ControlRigBlueprint /BRCosmetics/Accessories/FORT_Backpacks/Backpack_M_MED_FrozenTouch/Meshes/M_MED_FrozenTouch_Pack_CtrlRig
graphs [
 RigVMFunctionLibrary {links [], nodes []}
 RigVMModel {nodes [
   RigUnit_BeginExecution = &RigUnit_BeginExecution()
   RigVMFunction_Sequence = &RigVMFunction_Sequence(ExecuteContext=@RigUnit_BeginExecution.ExecutePin)
   Subtract = &RigVMFunction_MathDoubleSub(A=1, B=RigVMVariableNode() {variable=isCrouchingFloat}.Value)
   Absolute = &RigVMFunction_MathDoubleAbs(Value=@Subtract.Result)
   n1 = &RigVMFunction_AlphaInterp(InterpSpeedDecreasing=2, InterpSpeedIncreasing=2, Scale=1, bClampResult=False, bInterpResult=True, Value=@Absolute.Result)
   &RigUnit_ModifyTransforms(
    ItemToModify="((Item=(Type=Bone,Name=\"dyn_pack_spine_03\"),Transform=(Rotation=(X=0,Y=0.0436,Z=0,W=0.999),Translation=(X=0,Y=0,Z=1.5),Scale3D=(X=1,Y=1,Z=1))),(Item=(Type=Bone,Name=\"dyn_pack_spine_04\"),Transform=(Rotation=(X=0,Y=0.0436,Z=0,W=0.999),Translation=(X=0,Y=0,Z= ...(truncated)"
    WeightMinimum=0
    Weight=@n1.Result
    ExecutePin=@RigVMFunction_Sequence.A
   )
   AlphaInterp = &RigVMFunction_AlphaInterp(
    InterpSpeedDecreasing=4
    InterpSpeedIncreasing=4
    Scale=1
    bClampResult=False
    bInterpResult=True
    Value=RigVMVariableNode() {variable=isRunning}.Value
   )
   > &RigUnit_ModifyTransforms(
    ItemToModify="((Item=(Type=Bone,Name=\"dyn_pack_spine_03\"),Transform=(Rotation=(X=0,Y=-0.0436,Z=0,W=0.999),Translation=(X=0,Y=0,Z=0.5),Scale3D=(X=1,Y=1,Z=1))),(Item=(Type=Bone,Name=\"dyn_pack_spine_04\"),Transform=(Rotation=(X=0,Y=-0.0436,Z=0,W=0.999),Translation=(X=0,Y=0, ...(truncated)"
    WeightMinimum=0
    Weight=@AlphaInterp.Result
   )
   n2 = &RigVMFunction_AlphaInterp(
    InterpSpeedDecreasing=2
    InterpSpeedIncreasing=2
    Scale=1
    bClampResult=False
    bInterpResult=True
    Value=RigVMVariableNode() {variable=isTacticalSprinting}.Value
   )
   > &RigUnit_ModifyTransforms(
    ItemToModify="((Item=(Type=Bone,Name=\"dyn_pack_spine_03\"),Transform=(Rotation=(X=0,Y=-0.0436,Z=0,W=0.999),Translation=(X=0,Y=0,Z=1),Scale3D=(X=1,Y=1,Z=1))),(Item=(Type=Bone,Name=\"dyn_pack_spine_04\"),Transform=(Rotation=(X=0,Y=-0.0436,Z=0,W=0.999),Translation=(X=0,Y=0,Z= ...(truncated)"
    WeightMinimum=0
    Weight=@n2.Result
   )
  ]}
]
hierarchy [
 root {loc=<0 0 0>, rot=<0 0 0 1>}
 pack {parent=root, loc=<0 -0.7316 0>, rot=<0 0 0 1>}
 base_pack_spine {parent=pack, loc=<0 0.0464 10.1477>, rot=<-0.3747 -0.3747 0.5997 -0.5997>}
 dyn_pack_spine_01 {parent=base_pack_spine, loc=<0 0 0>, rot=<0 0 0 1>}
 dyn_pack_spine_02 {parent=dyn_pack_spine_01, loc=<5 0 0>, rot=<0 -0.1288 0 -0.9917>}
 dyn_pack_spine_03 {parent=dyn_pack_spine_02, loc=<5 0 0>, rot=<0 -0.1132 0 -0.9936>}
 dyn_pack_spine_04 {parent=dyn_pack_spine_03, loc=<4.6 0 -0.0147>, rot=<0 -0.0872 0 -0.9962>}
 dyn_pack_spine_05 {parent=dyn_pack_spine_04, loc=<4.7 0 0>, rot=<0 -0.1392 0 -0.9903>}
 dyn_pack_spine_06 {parent=dyn_pack_spine_05, loc=<4.4396 0 0>, rot=<0 0.1305 0 -0.9914>}
 dyn_pack_spine_07 {parent=dyn_pack_spine_06, loc=<3.3 0 0>, rot=<0 0.061 0 -0.9981>}
 dyn_pack_spine_08 {parent=dyn_pack_spine_07, loc=<6 0 0>, rot=<0 0 0 1>}
 body_root {parent=root, loc=<0 0 0>, rot=<0 0 0 1>}
 pelvis {parent=body_root, loc=<0 -0.7071 93.1191>, rot=<0.0127 0.707 -0.0127 -0.707>}
 spine_01 {parent=pelvis, loc=<6.7494 -0.6388 0>, rot=<0 0 0.034 -0.9994>}
 spine_02 {parent=spine_01, loc=<7.2518 -0.035 0>, rot=<0 0 0.0238 -0.9997>}
 spine_03 {parent=spine_02, loc=<6.0104 0 0>, rot=<0 0 0.0149 -0.9999>}
 spine_04 {parent=spine_03, loc=<8.1024 0 0>, rot=<0 0 -0.0132 -0.9999>}
 thigh_l {parent=pelvis, loc=<-2.5693 -0.0109 -7.7501>, rot=<0.0012 0.0572 0.0155 -0.9982>}
 calf_l {parent=thigh_l, loc=<-39.0484 0 0>, rot=<0.0269 -0.6882 -0.0283 -0.7245>}
 thigh_r {parent=pelvis, loc=<-2.5693 -0.0109 7.7501>, rot=<0.0572 -0.0012 0.9982 0.0155>}
 calf_r {parent=thigh_r, loc=<39.0484 0 0>, rot=<0.0006 -0.0161 -0.0391 -0.9991>}
]
variables [
 isCrouchingFloat {type=double, default=0}
 isRunning {type=double, default=0}
 isTacticalSprinting {type=double, default=0}
]
