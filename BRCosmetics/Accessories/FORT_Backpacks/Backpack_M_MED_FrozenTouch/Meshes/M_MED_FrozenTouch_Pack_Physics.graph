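# Physics asset for the Frozen Touch backpack cosmetic mesh, defining simple collision shapes on the thigh bones.
# It holds two bodies, thigh_l and thigh_r, each a single capsule of matching size, rotated and offset along local X in mirrored directions.
# It has no constraints, so the bodies are not jointed to each other, and no other bones have bodies.
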PhysicsAsset /BRCosmetics/Accessories/FORT_Backpacks/Backpack_M_MED_FrozenTouch/Meshes/M_MED_FrozenTouch_Pack_Physics
bodies {
 thigh_l [{shapeType=Capsule, length=10, radius=10, shapeName=thigh_l_capsule, center=<-19.5472 0 0>, rotation=<rot 0 -90 -90>}]
 thigh_r [{shapeType=Capsule, length=10, radius=10, shapeName=thigh_r_capsule, center=<19.9302 0 0>, rotation=<rot 0 -90 -90>}]
}
constraints []
